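# Sourdropper round status meeple widget: a resizable teammate icon showing one player's alive or dead state, tinted by team affiliation.
# SetAlive restores the live color, or on death either skips to dark or plays a scale-pulse animation.
# A looping 60 Hz timer converts animation time to AnimationAlpha, which GetColor uses to blend mid-bright, bright, then dark data-asset colors, defaulting to grey.
# Animation finish clears the timer unless revival interrupted it.
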WidgetBlueprint /Game/UI/Creative/Sourdropper/SourdropperRoundStatusMeeple
animations [DeathAnimation {
  binding_count=1
  length_seconds=1
  object=&self:DeathAnimation
  sequence {bindings [MainImage {
     id=#1
     possessed_class=Image
     tracks [{
       type=MovieScene2DTransformTrack
       sections [{
         type=MovieScene2DTransformSection
         end_frame=20
         start_frame=0
         channels [
          Translation.X {type=MovieSceneScriptingFloatChannel}
          Translation.Y {type=MovieSceneScriptingFloatChannel}
          Angle {type=MovieSceneScriptingFloatChannel}
          Scale.X {type=MovieSceneScriptingFloatChannel, keys [[0,1],[10,1.5],[20,1]]}
          Scale.Y {type=MovieSceneScriptingFloatChannel, keys [[0,1],[10,1.5],[20,1]]}
          Shear.X {type=MovieSceneScriptingFloatChannel}
          Shear.Y {type=MovieSceneScriptingFloatChannel}
         ]
        }]
      }]
    }]}
 }]
graphs [
 EventGraph {nodes [
   n1 = Event PreConstruct()
   n2 = Branch(execute=[@n1.then, @n3.then], Condition=Get bIgnorePostDeathAnimationSteps() -> (bIgnorePostDeathAnimationSteps:bool))
   Clear and Invalidate Timer by Handle(execute=@n2.else, Handle=Get ProcessDeathAnimationTimer() -> (ProcessDeathAnimationTimer:Timer Handle Structure))
   > Set AnimationAlpha(AnimationAlpha:double=1) -> (Output_Get:double)
   > Update()
   n3 = Event OnAnimationFinished!() -> (OutputDelegate, then, Animation:Widget Animation Object Reference) {is=title K2Node_Event}
  ]}
 GetColor {nodes [
   GetColor() {is=title K2Node_FunctionEntry}
   > n1 = Branch(Condition=GetColor!(
     TeamAffiliation:EFortTeamAffiliation Enum=Get TeamAffiliation() -> (TeamAffiliation:EFortTeamAffiliation Enum=Friendly)
     ColorType:EFortSourDropperColorType Enum=MidBright
     DataAsset:Fort Creative Sour Dropper Round Status Color Data Asset Object Reference=Get ColorData() -> (ColorData:Fort Creative Sour Dropper Round Status Color Data Asset Object Reference)
     OutColor:"Linear Color Structure (by ref)"=Get MidBrightColor() -> (MidBrightColor:Linear Color Structure)
    ) -> (bColorIsValid:bool) {is=title K2Node_CallFunction})
   n2 = Branch(execute=[@n1.then, @n11.then], Condition=Get bAlive() -> (bAlive:bool))
   > Return Node!(execute, Color:Linear Color Structure=Get MidBrightColor())
   n3 = Branch(
    execute=@n2.else
    Condition=GetColor(
     TeamAffiliation=Get TeamAffiliation() -> (TeamAffiliation=Friendly)
     ColorType=Bright
     DataAsset=Get ColorData()
     OutColor=Get BrightColor() -> (BrightColor:Linear Color Structure)
    )
   )
   n4 = Make Literal Float(Value=0.5)
   n5 = Branch(
    execute=[@n3.then, @n10.then]
    Condition="float < float"!(A:double=Get AnimationAlpha() -> (AnimationAlpha:double), B:double=@n4) -> (ReturnValue:bool) {is=title K2Node_CallFunction}
   )
   n6 = Get AnimationAlpha()
   > Return Node(Color="Lerp (LinearColor)"(A=Get MidBrightColor(), B=Get BrightColor(), Alpha=MapRangeClamped(Value=@n6, InRangeA=0, InRangeB=@n4)))
   n7 = Branch(
    execute=@n5.else
    Condition=GetColor(
     TeamAffiliation=Get TeamAffiliation() -> (TeamAffiliation=Friendly)
     ColorType=Dark
     DataAsset=Get ColorData()
     OutColor=Get DarkColor() -> (DarkColor:Linear Color Structure)
    )
   )
   n8 = Get AnimationAlpha()
   Return Node(
    execute=[@n7.then, @n9.then]
    Color="Lerp (LinearColor)"(A=Get BrightColor(), B=Get DarkColor(), Alpha=MapRangeClamped(Value=@n8, InRangeA=@n4, InRangeB=1))
   )
   n9 = Set DarkColor(execute=@n7.else, DarkColor:Linear Color Structure=Get DefaultColor() -> (DefaultColor:Linear Color Structure)) -> (Output_Get:Linear Color Structure)
   n10 = Set BrightColor(execute=@n3.else, BrightColor:Linear Color Structure=Get DefaultColor()) -> (Output_Get:Linear Color Structure)
   n11 = Set MidBrightColor(execute=@n1.else, MidBrightColor:Linear Color Structure=Get DefaultColor()) -> (Output_Get:Linear Color Structure)
  ]}
 ProcessDeathAnimation {nodes [
   ProcessDeathAnimation() {is=title K2Node_FunctionEntry}
   > n1 = Sequence()
   n2 = Get DeathAnimation() -> (DeathAnimation:Widget Animation Object Reference)
   n3 = GetEndTime(self=@n2)
   > Branch(Condition="float > float"(A=@n3, B=0) {is=title K2Node_CallFunction})
   > Set AnimationAlpha(AnimationAlpha=float / float(
     A=GetAnimationCurrentTime!(self:User Widget Object Reference, InAnimation:Widget Animation Object Reference=@n2) -> (ReturnValue:float)
     B=@n3
    ) {is=title K2Node_CallFunction})
   Update(execute=@n1.then_1)
  ]}
 SetAlive {nodes [
   n1 = SetAlive!() -> (then, bInAlive:bool=~, bSkipAnimations:bool=~) {is=title K2Node_FunctionEntry}
   > Branch(Condition="Not Equal (Boolean)"!(A:bool=@n1, B:bool=Get bAlive()) -> (ReturnValue:bool) {is=title K2Node_CallFunction})
   > n2 = Sequence()
   > Clear and Invalidate Timer by Handle(Handle=Get ProcessDeathAnimationTimer())
   > n3 = Set bAlive(bAlive:bool=Get bInAlive() -> (bInAlive:bool)) -> (Output_Get:bool)
   > n4 = Branch(Condition=@n3)
   > n5 = Branch(Condition=IsAnimationPlaying(InAnimation=Get DeathAnimation()))
   > Set bIgnorePostDeathAnimationSteps(bIgnorePostDeathAnimationSteps:bool=true) -> (Output_Get:bool)
   > n6 = StopAnimation(InAnimation=Get DeathAnimation())
   Set AnimationAlpha(execute=[@n6.then, @n5.else], AnimationAlpha=0)
   n7 = Branch(execute=@n4.else, Condition=Get bSkipAnimations() -> (bSkipAnimations:bool))
   > Set AnimationAlpha(AnimationAlpha=1)
   n8 = Branch(execute=@n7.else, Condition=IsAnimationPlaying(InAnimation=Get DeathAnimation()))
   Set AnimationAlpha(execute=@n8.else, AnimationAlpha=0)
   > Set bIgnorePostDeathAnimationSteps(bIgnorePostDeathAnimationSteps=false)
   > PlayAnimation(InAnimation=Get DeathAnimation())
   > n9 = Set Timer by Event(Delegate=Create Event(), Time=float / float(A=1, B=60), bLooping=true)
   > Set ProcessDeathAnimationTimer(ProcessDeathAnimationTimer:Timer Handle Structure=@n9) -> (Output_Get:Timer Handle Structure)
   Update(execute=@n2.then_1)
  ]}
 SetIconSize {nodes [
   n1 = SetIconSize!() -> (then, IconSize:Vector 2D Structure) {is=title K2Node_FunctionEntry}
   > n2 = Sequence()
   n3 = BreakVector2D(InVec=@n1)
   n4 = Get MainImageSizeBox() -> (MainImageSizeBox:Size Box Object Reference)
   > SetWidthOverride(self=@n4, InWidthOverride=@n3)
   SetHeightOverride(execute=@n2.then_1, self=@n4, InHeightOverride=@n3.Y)
  ]}
 SetTeamAffiliation {nodes [
   n1 = SetTeamAffiliation!() -> (then, InTeamAffiliation:EFortTeamAffiliation Enum=Friendly) {is=title K2Node_FunctionEntry}
   > Branch(Condition="Not Equal (Enum)"!(A:EFortTeamAffiliation Enum=@n1, B:EFortTeamAffiliation Enum=Get TeamAffiliation() -> (TeamAffiliation=Friendly)) -> (ReturnValue:bool))
   > Set TeamAffiliation(TeamAffiliation:EFortTeamAffiliation Enum=Get InTeamAffiliation() -> (InTeamAffiliation:EFortTeamAffiliation Enum=Friendly)) -> (Output_Get:EFortTeamAffiliation Enum=Friendly)
   > Update()
  ]}
 Update {nodes [
   Update() {is=title K2Node_FunctionEntry}
   > SetColorAndOpacity(
    self:Image Object Reference=Get MainImage() -> (MainImage:Image Object Reference)
    InColorAndOpacity:Linear Color Structure=GetColor!() -> (Color:Linear Color Structure)
   )
  ]}
]
variables [
 AnimationAlpha {type=number}
 bAlive {type=boolean}
 bIgnorePostDeathAnimationSteps {type=boolean}
 ColorData {type=FortCreativeSourDropperRoundStatusColorDataAsset, default=&/SourdropperColorData}
 DefaultColor {type=LinearColor, default=<rgba 0.5 0.5 0.5 1>}
 ProcessDeathAnimationTimer {type=TimerHandle}
 TeamAffiliation {type=EFortTeamAffiliation, default=Friendly}
]
widgets {
 info {inheritedWidgetCount=0, parentClass=&UserWidget, rootWidgetClass=&SizeBox, widgetCount=2}
 widgets [
  MainImageSizeBox {widgetClassPath=&SizeBox, bIsVariable=true, widget=MainImageSizeBox, properties {heightOverride=64, widthOverride=32}}
  MainImage {
   widgetClassPath=&Image
   bIsVariable=true
   parent=MainImageSizeBox
   widget=MainImage
   properties {brush {
     bIsDynamicallyLoaded=false
     drawAs=Image
     imageType=NoImage
     mirroring=NoMirror
     resourceName=None
     resourceObject=/Game/UI/Creative/Textures/Icon_Creative_HUD_Teammate
     tiling=NoTile
     imageSize=<32 32>
     margin=<ltrb 0 0 0 0>
     outlineSettings {
      bUseBrushTransparency=false
      roundingType=HalfHeightRadius
      width=0
      color {colorUseRule=UseColor_Specified, specifiedColor=<rgba 0 0 0 0>}
      cornerRadii=<0 0 0 0>
     }
     tintColor {colorUseRule=UseColor_Specified, specifiedColor=<rgba 1 1 1 1>}
     uVRegion {bIsValid=false, max=<0 0>, min=<0 0>}
    }}
   slot {class=SizeBoxSlot}
  }
 ]
}
