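# HUD widget for Sourdropper Creative minigames showing one side's round score and alive or eliminated player "meeples", relative to the view target.
# PreConstruct applies orientation (meeple order, backing angles), icon size and affiliation colors via a dynamic material.
# Team config updates refresh up to eight meeples and resize the root to the largest team.
# Friendly score is the bucket's round wins; otherwise it is completed rounds minus those wins.
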WidgetBlueprint /Game/UI/Creative/Sourdropper/SourdropperRoundStatusTeamStats
graphs [
 EventGraph {nodes [
   Event PreConstruct()
   > UpdateOrientation()
   > UpdateScoreImage()
   > UpdateMeepleIcons()
   Event Construct()
   n1 = Create Event()
   n2 = Get FortHUDContext!() -> (ReturnValue:Fort HUDContext Object Reference) {is=title K2Node_GetBlueprintContextEx}
   > Bind Event to On Local Player View Target Changed(self:Fort HUDContext Object Reference=@n2, Delegate=@n1) {is=title K2Node_AddDelegate}
   Event Destruct()
   > Unbind Event from On Local Player View Target Changed(self:Fort HUDContext Object Reference=@n2, Delegate=@n1) {is=title K2Node_RemoveDelegate}
   n3 = Event OnDisplayMinigameInProgressUI!() -> (OutputDelegate, then, bDisplay:bool) {is=title K2Node_Event}
   > Get!(execute:Exec) -> (then:Exec, else:Exec, CurrentMinigame:Fort Minigame Object Reference)
   > n4 = Branch(Condition=@n3.bDisplay)
   n5 = Create Event()
   > Assign Minigame Team Config Updated!(
    execute:Exec
    self:Fort Minigame Object Reference=Get CurrentMinigame() -> (CurrentMinigame:Fort Minigame Object Reference)
    Delegate:"Delegate (by ref)"=@n5
   ) -> (then:Exec) {is=title K2Node_AssignDelegate}
   > OnMinigameTeamConfigUpdated()
   Unbind Event from Minigame Team Config Updated(execute=@n4.else, self:Fort Minigame Object Reference=Get CurrentMinigame(), Delegate=@n5) {is=title K2Node_RemoveDelegate}
  ]}
 IsTeamIndexRelevant {nodes [
   IsTeamIndexRelevant!() -> (then, TeamIndex:Byte, OwningTeamIndex:Byte, TeamSize:Byte) {is=title K2Node_FunctionEntry}
   n1 = Get TeamSize() -> (TeamSize:Byte=0)
   n2 = Get TeamIndex() -> (TeamIndex:Byte=0)
   n3 = Get OwningTeamIndex() -> (OwningTeamIndex:Byte=0)
   n4 = "Not Equal (Byte)"!(A:Byte=@n2, B:Byte=@n3) -> (ReturnValue:bool)
   n5 = "Equal (Byte)"!(A:Byte=@n2, B:Byte=@n3) -> (ReturnValue:bool)
   n6 = "Equal (Enum)"(
    A:EFortTeamAffiliation Enum=Get TeamAffiliation() -> (TeamAffiliation:EFortTeamAffiliation Enum=Friendly)
    B:EFortTeamAffiliation Enum=Friendly
   )
   > Return Node!(
    execute
    ReturnValue:bool=AND Boolean(
     A=Select(Option 0:bool=@n4, Option 1:bool=@n5, Index=@n6) -> (ReturnValue:bool)
     B="Byte > Byte"!(A:Byte=@n1, B:Byte=0) -> (ReturnValue:bool)
    )
   )
  ]}
 OnMinigameTeamConfigUpdated {nodes [
   OnMinigameTeamConfigUpdated() {is=title K2Node_FunctionEntry}
   > Update()
  ]}
 Update {nodes [
   Update() {is=title K2Node_FunctionEntry}
   n1 = GetOwningPlayer()
   > Is Valid(InputObject=@n1)
   > n2 = Cast To Pawn!(execute, Object=GetViewTarget!(self:Controller Object Reference=@n1) -> (ReturnValue:Actor Object Reference)) -> (then, CastFailed, AsPawn:Pawn Object Reference) {is=title K2Node_DynamicCast}
   > n3 = Sequence()
   n4 = GetActorTeam!(Actor:Actor Object Reference=@n2) -> (ReturnValue:Byte=0)
   > Branch(Condition="Equal (Byte)"(A=Get ViewTargetTeam() -> (ViewTargetTeam:Byte=0), B=@n4))
   > Set ViewTargetTeam(ViewTargetTeam:Byte=@n4) -> (Output_Get:Byte=0)
   > "Set bSkipMeepleAnimations "("bSkipMeepleAnimations ":bool=true) -> (Output_Get:bool)
   n5 = Sequence(execute=@n3.then_1)
   > UpdateMeeples!(
    execute
    self
    ViewTargetPawn:Pawn Object Reference=@n2
    bSkipMeepleAnimations:bool="Get bSkipMeepleAnimations "() -> ("bSkipMeepleAnimations ":bool)
   ) -> (then)
   UpdateScoreText!(execute=@n5.then_1, self, ViewTargetPawn:Pawn Object Reference=@n2) -> (then)
  ]}
 UpdateMeepleIcons {nodes [
   UpdateMeepleIcons() {is=title K2Node_FunctionEntry}
   > n1 = For Each Loop(Array:Array of Sourdropper Round Status Meeple Object References=Get MeepleWidgets() -> (MeepleWidgets:Array of Sourdropper Round Status Meeple Object References)) -> (Array Element:Sourdropper Round Status Meeple Object Reference)
   > Is Valid(InputObject=@n1)
   > SetIconSize!(
    execute
    self:Sourdropper Round Status Meeple Object Reference=@n1
    IconSize:Vector 2D Structure=Get MeepleIconSize() -> (MeepleIconSize:Vector 2D Structure)
   ) -> (then)
  ]}
 UpdateMeeples {nodes [
   n1 = UpdateMeeples!() -> (then, ViewTargetPawn:Pawn Object Reference, bSkipMeepleAnimations:bool=~) {is=title K2Node_FunctionEntry}
   > Is Valid(InputObject=@n1)
   > Set OwningTeamIndex(OwningTeamIndex:Byte=GetActorTeam(Actor=@n1) -> (ReturnValue=0)) -> (Output_Get:Byte=0)
   > n2 = Get()
   n3 = GetMinigameTeamComponent!(self:Fort Minigame Object Reference=@n2) -> (ReturnValue:Fort Minigame Team Component Object Reference)
   > Is Valid(InputObject=@n3)
   > n4 = Sequence!(execute) -> (then_0, then_1, then_2:Exec)
   > n5 = For Each Loop(Array=Get MeepleWidgets())
   > SetVisibility(self=@n5, InVisibility=Collapsed)
   n6 = "To Integer (Byte)"!(InByte:Byte=Get MaxTeamSize() -> (MaxTeamSize:Byte=0)) -> (ReturnValue:int=0)
   n7 = float + float(A=BreakVector2D(InVec=Get MeepleIconSize()), B=4) {is=title K2Node_CommutativeAssociativeBinaryOperator}
   SetWidthOverride(
    execute=@n4.then_2
    self=Get RootSizeBox() -> (RootSizeBox:Size Box Object Reference)
    InWidthOverride="Max (Float)"(A=int * float!(A:int=@n6, B:double=@n7) -> (ReturnValue:double), B=float * float(A=@n7, B=2))
   )
   n8 = For Each Loop with Break!(
    Exec:Exec=@n4.then_1
    Array:Array of Minigame Team Structures=Get TeamArray!(self:Fort Minigame Team Component Object Reference=@n3) -> (TeamArray:Array of Minigame Team Structures)
    Break:Exec=@n10.Completed
   ) -> (
    LoopBody:Exec
    Array Element_TeamIndex:Byte
    Array Element_TeamName:Text
    Array Element_TeamDescription:Text
    Array Element_TeamColorIndex:int
    Array Element_MaxInitTeamSize:int
    Array Element_InitTeamSizeWeight:int
    Array Element_bHasBucketAvailable:bool=~
    Array Element_EliminatedCount:Byte
    Array Element_TeamSize:Byte
    Array Element_TeamIcon:Creative Icon Option Structure
    Array Element_bUseTeamScore:bool=~
    Array Index:int
    Completed:Exec
   )
   > Set MaxTeamSize(MaxTeamSize:Byte="Max (Byte)"(A:Byte=Get MaxTeamSize() -> (MaxTeamSize=0), B:Byte=@n8.Array Element_TeamSize) -> (ReturnValue:Byte=0) {is=title K2Node_CommutativeAssociativeBinaryOperator}) -> (Output_Get:Byte=0)
   > Branch(Condition=IsTeamIndexRelevant!(TeamIndex:Byte=@n8, OwningTeamIndex:Byte=Get OwningTeamIndex() -> (OwningTeamIndex=0), TeamSize:Byte=@n8.Array Element_TeamSize) -> (ReturnValue:bool) {is=title K2Node_CallFunction})
   n9 = "To Integer (Byte)"(InByte=@n8.Array Element_TeamSize) -> (ReturnValue=0)
   > Set AliveCount(AliveCount:int=int - int(A=@n9, B="To Integer (Byte)"(InByte=@n8.Array Element_EliminatedCount) -> (ReturnValue=0)) {is=title K2Node_CallFunction}) -> (Output_Get:int=0)
   > n10 = For Loop(LastIndex=int - int(A=@n9, B=1))
   > n11 = Set MeepleWidget(MeepleWidget:Sourdropper Round Status Meeple Object Reference="Get (a copy)"(Array:Array of Sourdropper Round Status Meeple Object References=Get MeepleWidgets(), Dimension 1=@n10) -> (Output:Sourdropper Round Status Meeple Object Reference)) -> (Output_Get:Sourdropper Round Status Meeple Object Reference)
   > SetVisibility(self=@n11, InVisibility=SelfHitTestInvisible)
   > SetAlive!(
    execute
    self:Sourdropper Round Status Meeple Object Reference=Get MeepleWidget() -> (MeepleWidget:Sourdropper Round Status Meeple Object Reference)
    bInAlive:bool="integer < integer"!(A:int=@n10, B:int=Get AliveCount() -> (AliveCount:int=0)) -> (ReturnValue:bool) {is=title K2Node_CallFunction}
    bSkipAnimations:bool=Get bSkipMeepleAnimations() -> (bSkipMeepleAnimations:bool)
   ) -> (then)
  ]}
 UpdateOrientation {nodes [
   UpdateOrientation() {is=title K2Node_FunctionEntry}
   > n1 = Sequence()
   Clear(execute=@n1.then_2, TargetArray:Array of Sourdropper Round Status Meeple Object References=Get MeepleWidgets())
   > n2 = Branch(Condition="Equal (Enum)"(
     A:E_SourdropperTeamStatsOrientation Enum=Get Orientation() -> (Orientation:E_SourdropperTeamStatsOrientation Enum=NewEnumerator0)
     B:E_SourdropperTeamStatsOrientation Enum=NewEnumerator0
    ))
   > n3 = Set MeepleWidgets(MeepleWidgets:Array of Sourdropper Round Status Meeple Object References=Make Array!(
     "[0]":Sourdropper Round Status Meeple Object Reference=Get Meeple_7() -> (Meeple_7:Sourdropper Round Status Meeple Object Reference)
     "[1]":Sourdropper Round Status Meeple Object Reference=Get Meeple_6() -> (Meeple_6:Sourdropper Round Status Meeple Object Reference)
     "[2]":Sourdropper Round Status Meeple Object Reference=Get Meeple_5() -> (Meeple_5:Sourdropper Round Status Meeple Object Reference)
     "[3]":Sourdropper Round Status Meeple Object Reference=Get Meeple_4() -> (Meeple_4:Sourdropper Round Status Meeple Object Reference)
     "[4]":Sourdropper Round Status Meeple Object Reference=Get Meeple_3() -> (Meeple_3:Sourdropper Round Status Meeple Object Reference)
     "[5]":Sourdropper Round Status Meeple Object Reference=Get Meeple_2() -> (Meeple_2:Sourdropper Round Status Meeple Object Reference)
     "[6]":Sourdropper Round Status Meeple Object Reference=Get Meeple_1() -> (Meeple_1:Sourdropper Round Status Meeple Object Reference)
     "[7]":Sourdropper Round Status Meeple Object Reference=Get Meeple_0() -> (Meeple_0:Sourdropper Round Status Meeple Object Reference)
    ) -> (Array:Array of Sourdropper Round Status Meeple Object References)) -> (Output_Get:Array of Sourdropper Round Status Meeple Object References)
   n4 = For Each Loop(Exec=[@n3.then, @n5.then], Array=Get MeepleWidgets())
   > SetTeamAffiliation!(
    execute
    self:Sourdropper Round Status Meeple Object Reference=@n4
    InTeamAffiliation:EFortTeamAffiliation Enum=Get TeamAffiliation() -> (TeamAffiliation=Friendly)
   ) -> (then)
   n5 = Set MeepleWidgets(
    execute=@n2.else
    MeepleWidgets=Make Array(
     "[0]"=Get Meeple_0()
     "[1]"=Get Meeple_1()
     "[2]"=Get Meeple_2()
     "[3]"=Get Meeple_3()
     "[4]"=Get Meeple_4()
     "[5]"=Get Meeple_5()
     "[6]"=Get Meeple_6()
     "[7]"=Get Meeple_7()
    )
   )
  ]}
 UpdateScoreImage {nodes [
   UpdateScoreImage() {is=title K2Node_FunctionEntry}
   > n1 = GetDynamicMaterial(self=Get ScoreBackgroundImage() -> (ScoreBackgroundImage:Image Object Reference))
   > Is Valid(InputObject=@n1)
   n2 = Get TeamAffiliation() -> (TeamAffiliation=Friendly)
   n3 = Get ColorData() -> (ColorData:Fort Creative Sour Dropper Round Status Color Data Asset Object Reference)
   n4 = Get RGBA1Color() -> (RGBA1Color:Linear Color Structure)
   > SetVectorParameterValue(
    self=@n1
    ParameterName=RGBA1
    Value=Select(
     Option 0:Linear Color Structure=Get DefaultColor() -> (DefaultColor:Linear Color Structure)
     Option 1:Linear Color Structure=Get RGBA1Color()
     Index=GetColor!(
      TeamAffiliation:EFortTeamAffiliation Enum=@n2
      ColorType:EFortSourDropperColorType Enum=MidDark
      DataAsset:Fort Creative Sour Dropper Round Status Color Data Asset Object Reference=@n3
      OutColor:"Linear Color Structure (by ref)"=@n4
     ) -> (bColorIsValid:bool)
    ) -> (ReturnValue:Linear Color Structure)
   )
   n5 = Get TeamAffiliation() -> (TeamAffiliation=Friendly)
   n6 = Get RGBA2Color() -> (RGBA2Color:Linear Color Structure)
   n7 = Get ColorData()
   > SetVectorParameterValue(
    self=@n1
    ParameterName=RGBA2
    Value=Select(
     Option 0=Get DefaultColor()
     Option 1=Get RGBA2Color()
     Index=GetColor(TeamAffiliation=@n5, ColorType=MidBright, DataAsset=@n7, OutColor=@n6)
    )
   )
   > SetScalarParameterValue(
    self=@n1
    ParameterName=LeftAngle
    Value=Select!(
     NewEnumerator0:double=10
     NewEnumerator1:double=-3
     Index:E_SourdropperTeamStatsOrientation Enum=Get Orientation() -> (Orientation=NewEnumerator0)
    ) -> (ReturnValue:double)
   )
   > SetScalarParameterValue(
    self=@n1
    ParameterName=RightAngle
    Value=Select(NewEnumerator0=3, NewEnumerator1=-10, Index=Get Orientation() -> (Orientation=NewEnumerator0))
   )
  ]}
 UpdateScoreText {nodes [
   UpdateScoreText!() -> (then, ViewTargetPawn:Pawn Object Reference) {is=title K2Node_FunctionEntry}
   > n1 = Sequence!(execute) -> (then_0, then_1)
   > n2 = Get!(execute) -> (then, else, ViewTargetPawn:Pawn Object Reference)
   > n3 = Get!(execute) -> (then, else, CurrentMinigame:Fort Minigame Object Reference)
   n4 = GetMinigameTeamComponent(self=@n3)
   > Is Valid(InputObject=@n4)
   n5 = Get Round Wins For Bucket Index!(
    self:Fort Minigame Team Component Object Reference=@n4
    BucketIndex:int=GetBucketForPlayerState!(
     self:Fort Minigame Team Component Object Reference=@n4
     PlayerState:Player State Object Reference=Get PlayerState!(self:Pawn Object Reference=@n2) -> (PlayerState:Player State Object Reference)
     bMatchCurrentTeam:bool=false
    ) -> (ReturnValue:int=0)
   ) -> (WinCount:int=0, ReturnValue:bool) {is=title K2Node_CallFunction}
   > Branch(Condition=@n5.ReturnValue)
   > n6 = Set BucketWinCount(BucketWinCount:int=@n5) -> (Output_Get:int=0)
   n7 = int - int(A=Get CurrentRound!(self:Fort Minigame Object Reference=Get CurrentMinigame()) -> (CurrentRound:int=0), B=1)
   n8 = Get TeamAffiliation() -> (TeamAffiliation=Friendly)
   > Set Score(Score:int=Select!(
     Option 0:int=int - int(A=@n7, B=@n6)
     Option 1:int=@n6
     Index:bool="Equal (Enum)"(A:EFortTeamAffiliation Enum=@n8, B:EFortTeamAffiliation Enum=Friendly)
    ) -> (ReturnValue:int)) -> (Output_Get:int=0)
   "SetText (Text)"(
    execute=@n1.then_1
    self=Get ScoreTextblock() -> (ScoreTextblock:Common Text Object Reference)
    InText="To Text (Integer)"(Value=Get Score() -> (Score:int=0), bUseGrouping=false)
   )
  ]}
]
variables [
 ColorData {type=FortCreativeSourDropperRoundStatusColorDataAsset, default=&/SourdropperColorData}
 DefaultColor {type=LinearColor, default=<rgba 0.5 0.5 0.5 1>}
 MeepleIconSize {type=Vector2D, default=<32 32>}
 MeepleWidgets {type="array<SourdropperRoundStatusMeeple_C>", default []}
 Orientation {type=E_SourdropperTeamStatsOrientation, default=Left}
 TeamAffiliation {type=EFortTeamAffiliation, default=Friendly}
 ViewTargetTeam {type=integer}
]
widgets {
 info {inheritedWidgetCount=1, parentClass=/Script/FortniteUI.MinigameWidgetBase, rootWidgetClass=&SizeBox, widgetCount=15}
 widgets [
  RootSizeBox {widgetClassPath=&SizeBox, bIsVariable=true, widget=RootSizeBox, properties {bOverride_MinDesiredWidth=true, minDesiredWidth=72}}
  VerticalBox_0 {widgetClassPath=&VerticalBox, parent=RootSizeBox, widget=VerticalBox_0, slot {class=SizeBoxSlot}}
  ScoreOverlay {widgetClassPath=&Overlay, bIsVariable=true, parent=VerticalBox_0, widget=ScoreOverlay, slot {class=VerticalBoxSlot}}
  ScoreBackgroundImage {
   widgetClassPath=&Image
   bIsVariable=true
   parent=ScoreOverlay
   widget=ScoreBackgroundImage
   properties {brush {
     bIsDynamicallyLoaded=false
     drawAs=Box
     imageType=NoImage
     mirroring=NoMirror
     resourceName=None
     resourceObject=/Game/UI/Creative/Materials/M_UI_Creative_HUD_ScoreBacking
     tiling=NoTile
     imageSize=<32 32>
     margin=<ltrb 0.5 0.5 0.5 0.5>
     outlineSettings {
      bUseBrushTransparency=false
      roundingType=HalfHeightRadius
      width=0
      color {colorUseRule=UseColor_Specified, specifiedColor=<rgba 0 0 0 0>}
      cornerRadii=<0 0 0 0>
     }
     tintColor {colorUseRule=UseColor_Specified, specifiedColor=<rgba 1 1 1 1>}
     uVRegion {bIsValid=false, max=<0 0>, min=<0 0>}
    }}
   slot {class=OverlaySlot, properties {horizontalAlignment=HAlign_Fill, verticalAlignment=VAlign_Center}}
  }
  ScoreTextblock {
   widgetClassPath=&CommonTextBlock
   bIsVariable=true
   parent=ScoreOverlay
   widget=ScoreTextblock
   properties {
    style=/FortUILibrary/Styles/Typography/Foundation/TextStyle_BurbankSmall_S_White.TextStyle_BurbankSmall_S_White_C
    text=5
    font {
     bForceMonospaced=false
     bMaterialIsStencil=false
     fontMaterial=null
     fontObject=&BurbankSmall
     letterSpacing=0
     monospacedWidth=1
     size=18
     skewAmount=0
     typefaceFontName=Black
     outlineSettings {
      bApplyOutlineToDropShadows=false
      bMiteredCorners=false
      bSeparateFillAlpha=false
      outlineMaterial=null
      outlineSize=0
      outlineColor=<rgba 0 0 0 1>
     }
    }
    shadowOffset=<0 0>
   }
   slot {class=OverlaySlot, properties {horizontalAlignment=HAlign_Center, verticalAlignment=VAlign_Center}}
  }
  MeepleGrid {
   widgetClassPath=&HorizontalBox
   bIsVariable=true
   parent=VerticalBox_0
   widget=MeepleGrid
   slot {class=VerticalBoxSlot, properties {horizontalAlignment=HAlign_Center}}
  }
  Meeple_0 {
   widgetClassPath=&/SourdropperRoundStatusMeeple
   bIsVariable=true
   parent=MeepleGrid
   widget=Meeple_0
   properties {visibility=Collapsed}
   slot {class=HorizontalBoxSlot, properties {verticalAlignment=VAlign_Center, padding=<ltrb 2 0 2 0>}}
  }
  Meeple_1 {
   widgetClassPath=&SourdropperRoundStatusMeeple
   bIsVariable=true
   parent=MeepleGrid
   widget=Meeple_1
   properties {visibility=Collapsed}
   slot {class=HorizontalBoxSlot, properties {verticalAlignment=VAlign_Center, padding=<ltrb 2 0 2 0>}}
  }
  Meeple_2 {
   widgetClassPath=&SourdropperRoundStatusMeeple
   bIsVariable=true
   parent=MeepleGrid
   widget=Meeple_2
   properties {visibility=Collapsed}
   slot {class=HorizontalBoxSlot, properties {verticalAlignment=VAlign_Center, padding=<ltrb 2 0 2 0>}}
  }
  Meeple_3 {
   widgetClassPath=&SourdropperRoundStatusMeeple
   bIsVariable=true
   parent=MeepleGrid
   widget=Meeple_3
   properties {visibility=Collapsed}
   slot {class=HorizontalBoxSlot, properties {verticalAlignment=VAlign_Center, padding=<ltrb 2 0 2 0>}}
  }
  Meeple_4 {
   widgetClassPath=&SourdropperRoundStatusMeeple
   bIsVariable=true
   parent=MeepleGrid
   widget=Meeple_4
   properties {visibility=Collapsed}
   slot {class=HorizontalBoxSlot, properties {verticalAlignment=VAlign_Center, padding=<ltrb 2 0 2 0>}}
  }
  Meeple_5 {
   widgetClassPath=&SourdropperRoundStatusMeeple
   bIsVariable=true
   parent=MeepleGrid
   widget=Meeple_5
   properties {visibility=Collapsed}
   slot {class=HorizontalBoxSlot, properties {verticalAlignment=VAlign_Center, padding=<ltrb 2 0 2 0>}}
  }
  Meeple_6 {
   widgetClassPath=&SourdropperRoundStatusMeeple
   bIsVariable=true
   parent=MeepleGrid
   widget=Meeple_6
   properties {visibility=Collapsed}
   slot {class=HorizontalBoxSlot, properties {verticalAlignment=VAlign_Center, padding=<ltrb 2 0 2 0>}}
  }
  Meeple_7 {
   widgetClassPath=&SourdropperRoundStatusMeeple
   bIsVariable=true
   parent=MeepleGrid
   widget=Meeple_7
   properties {visibility=Collapsed}
   slot {class=HorizontalBoxSlot, properties {verticalAlignment=VAlign_Center, padding=<ltrb 2 0 2 0>}}
  }
  TutorialHighlight {widgetClassPath=&TutorialHighlight}
 ]
}
