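# Minigame HUD widget for Sourdropper in Creative, showing the current round number between friendly and hostile team stats.
# On the minigame in-progress display event, it collapses itself if display is off, no minigame exists, it is a curated hub, or game-state HUD-hide tags match.
# While visible, it binds to the minigame's round-change delegate and refreshes the round text immediately.
# When hidden, it unbinds that delegate.
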WidgetBlueprint /Game/UI/Creative/Sourdropper/SourdropperRoundStatus
graphs [
 EventGraph {nodes [
   n1 = Event OnDisplayMinigameInProgressUI!() -> (OutputDelegate, then, bDisplay:bool) {is=title K2Node_Event}
   > n2 = Sequence()
   > n3 = Branch(Condition=@n1.bDisplay)
   > n4 = Get!(execute:Exec) -> (then:Exec, else:Exec, CurrentMinigame:Fort Minigame Object Reference)
   n5 = Literal enum ESlateVisibility!(Enum:ESlateVisibility Enum=Collapsed) -> (ReturnValue:ESlateVisibility Enum) {is=title K2Node_EnumLiteral}
   > Set DesiredVisibility(DesiredVisibility:ESlateVisibility Enum=Select(Option 0=HitTestInvisible, Option 1=@n5, Index=IsCuratedHub!(self:Fort Minigame Object Reference=@n4) -> (ReturnValue:bool))) -> (Output_Get:ESlateVisibility Enum=Visible)
   n6 = GetGameStateAthenaFromContext!(WorldContextObject:Object Reference=@n4) -> (ReturnValue:Fort Game State Athena Object Reference)
   > Is Valid(InputObject=@n6)
   n7 = GetOwningPlayerPawn!(self:User Widget Object Reference) -> (ReturnValue:Pawn Object Reference)
   > Branch(Condition=HasAnyTags!(
     self:Fort HUDElement Widget Object Reference
     InTags:"Gameplay Tag Container Structure (by ref)"=GetHUDElementsToHide!(self:Fort Game State Object Reference=@n6, Pawn:Pawn Object Reference=@n7) -> (ReturnValue:Gameplay Tag Container Structure)
    ) -> (ReturnValue:bool))
   > Set DesiredVisibility(DesiredVisibility=Collapsed) -> (Output_Get=Visible)
   Set DesiredVisibility(execute=[@n4.else, @n3.else], DesiredVisibility=@n5) -> (Output_Get=Visible)
   SetVisibility(execute=@n2.then_1, InVisibility=Get DesiredVisibility() -> (DesiredVisibility:ESlateVisibility Enum=Visible))
   > Get()
   > n8 = Branch(Condition=IsVisible())
   n9 = Create Event()
   > Bind Event to On Current Round Changed Delegate(self:Fort Minigame Object Reference=Get CurrentMinigame() -> (CurrentMinigame:Fort Minigame Object Reference), Delegate=@n9) {is=title K2Node_AddDelegate}
   n10 = Get CurrentMinigame()
   > OnCurrentRoundChanged!(
    execute
    self
    Minigame:Fort Minigame Object Reference=@n10
    CurrentRound:int=Get CurrentRound!(self:Fort Minigame Object Reference=@n10) -> (CurrentRound:int=0)
   ) -> (then)
   Unbind Event from On Current Round Changed Delegate(execute=@n8.else, self:Fort Minigame Object Reference=Get CurrentMinigame(), Delegate=@n9) {is=title K2Node_RemoveDelegate}
  ]}
 OnCurrentRoundChanged {nodes [
   n1 = OnCurrentRoundChanged!() -> (then, Minigame:Fort Minigame Object Reference, CurrentRound:int) {is=title K2Node_FunctionEntry}
   > "SetText (Text)"(
    self=Get RoundNumberTextblock() -> (RoundNumberTextblock:Common Text Object Reference)
    InText=Format Text!(Format="LOCTEXT(\"Round {Number}\")", Number:Text="To Text (Integer)"(Value=@n1.CurrentRound, bUseGrouping=false)) -> (Result)
   )
  ]}
]
variables [DesiredVisibility {type=ESlateVisibility, default=SelfHitTestInvisible}]
widgets {
 info {inheritedWidgetCount=1, parentClass=/Script/FortniteUI.MinigameWidgetBase, rootWidgetClass=&Overlay, widgetCount=7}
 widgets [
  Overlay_2 {widgetClassPath=&Overlay, widget=Overlay_2}
  Overlay_0 {
   widgetClassPath=&Overlay
   parent=Overlay_2
   widget=Overlay_0
   slot {class=OverlaySlot, properties {horizontalAlignment=HAlign_Center}}
  }
  HorizontalBox_40 {
   widgetClassPath=&HorizontalBox
   parent=Overlay_0
   widget=HorizontalBox_40
   slot {class=OverlaySlot, properties {horizontalAlignment=HAlign_Center, verticalAlignment=VAlign_Center}}
  }
  FriendlyTeam {
   widgetClassPath=&/SourdropperRoundStatusTeamStats
   bIsVariable=true
   parent=HorizontalBox_40
   widget=FriendlyTeam
   properties {hiddenVisibility=Collapsed}
   slot {class=HorizontalBoxSlot, properties {verticalAlignment=VAlign_Center}}
  }
  RoundNumberTextblock {
   widgetClassPath=&CommonTextBlock
   bIsVariable=true
   parent=HorizontalBox_40
   widget=RoundNumberTextblock
   properties {
    style=/FortUILibrary/Styles/Typography/Foundation/TextStyle-Header-S-outline.TextStyle-Header-S-outline_C
    text=Round 7
    textTransformPolicy=ToUpper
    font {
     bForceMonospaced=false
     bMaterialIsStencil=false
     fontMaterial=null
     fontObject=&Burbank
     letterSpacing=0
     monospacedWidth=1
     size=24
     skewAmount=0
     typefaceFontName=Burbank-Black
     outlineSettings {
      bApplyOutlineToDropShadows=false
      bMiteredCorners=false
      bSeparateFillAlpha=false
      outlineMaterial=null
      outlineSize=0
      outlineColor=<rgba 0 0 0 1>
     }
    }
    shadowColorAndOpacity=<rgba 0 0 0 1>
    shadowOffset=<1.6 1.6>
   }
   slot {class=HorizontalBoxSlot, properties {verticalAlignment=VAlign_Center, padding=<ltrb 10 10 10 10>}}
  }
  HostileTeam {
   widgetClassPath=&SourdropperRoundStatusTeamStats
   bIsVariable=true
   parent=HorizontalBox_40
   widget=HostileTeam
   properties {hiddenVisibility=Collapsed, orientation=Right, teamAffiliation=Hostile}
   slot {class=HorizontalBoxSlot, properties {horizontalAlignment=HAlign_Left, verticalAlignment=VAlign_Center}}
  }
  TutorialHighlight {widgetClassPath=&TutorialHighlight}
 ]
}
