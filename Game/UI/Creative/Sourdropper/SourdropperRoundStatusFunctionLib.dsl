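; Blueprint function library for the Creative SourDropper round status UI, providing a GetColor lookup for team-relative colors.
; GetColor checks the color data asset, finds the entry for the team affiliation in its AffiliationColors map, then looks up the color type in that entry's nested map.
; A hit writes the by-ref OutColor and returns true.
; A Sequence's second branch returns false as the fallback, leaving OutColor untouched.
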(fn GetColor
  (params (TeamAffiliation :type EFortTeamAffiliation%20Enum :default "Friendly") (ColorType :type EFortSourDropperColorType%20Enum) (DataAsset :type Fort%20Creative%20Sour%20Dropper%20Round%20Status%20Color%20Data%20Asset%20Object%20Reference) (OutColor :type Linear%20Color%20Structure%20%28by%20ref%29))
  (bind _teamaffiliation (Variables|GetTeamAffiliation))
  (bind _colortype (Variables|GetColorType))
  (Utilities|FlowControl|Sequence
    (:then_0
      (bind _dataasset (Variables|GetDataAsset)
        (:then
          (if (.ReturnValue (Utilities|Map|Find (Variables|/Script/FortniteGame.FortCreativeSourDropperRoundStatusColorDataAsset|GetAffiliationColors _dataasset) _teamaffiliation))
            (if (.ReturnValue (Utilities|Map|Find (Utilities|Struct|BreakFortSourDropperColorData (Utilities|Map|Find (Variables|/Script/FortniteGame.FortCreativeSourDropperRoundStatusColorDataAsset|GetAffiliationColors _dataasset) _teamaffiliation)) _colortype))
              (Variables|SetBy-RefVar (Variables|GetOutColor) (Utilities|Map|Find (Utilities|Struct|BreakFortSourDropperColorData (Utilities|Map|Find (Variables|/Script/FortniteGame.FortCreativeSourDropperRoundStatusColorDataAsset|GetAffiliationColors _dataasset) _teamaffiliation)) _colortype))
              (return true))))
        (:else)))
    (:then_1
      (return false))))
